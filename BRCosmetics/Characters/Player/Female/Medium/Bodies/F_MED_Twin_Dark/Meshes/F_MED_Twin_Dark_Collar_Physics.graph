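# Physics asset for the collar of the F_MED_Twin_Dark female player body, covering left and right collar pieces and a loop.
# Each piece is a capsule body constrained to a tiny anchor capsule on a parent or driver bone.
# Collar sides share tight swing and twist limits, while the loop swings wider with almost no twist.
# The left collar links to dyn_driver_collarA_l, leaving the dyn_parent_collar_l body without a constraint.
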PhysicsAsset /BRCosmetics/Characters/Player/Female/Medium/Bodies/F_MED_Twin_Dark/Meshes/F_MED_Twin_Dark_Collar_Physics
bodies {
 dyn_collarA_01_l [{shapeType=Capsule, length=8, radius=3.25, center=<0.35 0 0>, rotation=<rot 90 0 0>}]
 dyn_collar_01_r [{shapeType=Capsule, length=8, radius=3.25, center=<-0.35 0 0>, rotation=<rot 90 0 0>}]
 dyn_driver_collarA_l [{shapeType=Capsule, length=0.505, radius=0.505, rotation=<rot 0 -90 0>}]
 dyn_loop_01 [{shapeType=Capsule, length=2.5, radius=3.75, center=<1.9914 0 0>, rotation=<rot 90 0 0>}]
 dyn_parent_collar_l [{shapeType=Capsule, length=0.505, radius=0.505, rotation=<rot 0 -90 0>}]
 dyn_parent_collar_r [{shapeType=Capsule, length=0.505, radius=0.505, rotation=<rot 0 -90 0>}]
 dyn_parent_loop [{shapeType=Capsule, length=0.505, radius=0.505, rotation=<rot 0 -90 0>}]
}
constraints [
 {bone1Name=dyn_collar_01_r, bone2Name=dyn_parent_collar_r, swing1LimitDegrees=15, swing2LimitDegrees=15, twistLimitDegrees=5}
 {bone1Name=dyn_loop_01, bone2Name=dyn_parent_loop, swing1LimitDegrees=35, swing2LimitDegrees=45, twistLimitDegrees=1}
 {bone1Name=dyn_collarA_01_l, bone2Name=dyn_driver_collarA_l, swing1LimitDegrees=15, swing2LimitDegrees=15, twistLimitDegrees=5}
]
