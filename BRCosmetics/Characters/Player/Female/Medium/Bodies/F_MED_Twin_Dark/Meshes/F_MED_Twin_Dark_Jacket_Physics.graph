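# Physics asset for the jacket of the F_MED_Twin_Dark female player body, simulating the center, left and right back flaps of the shirt bones.
# Each flap is a single capsule constrained to a tiny capsule on its own parent bone, with moderate swing limits and a tight twist limit.
# The left and right flaps also link to the center flap through constraints with free swing and twist, coupling the three without restricting their relative rotation.
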PhysicsAsset /BRCosmetics/Characters/Player/Female/Medium/Bodies/F_MED_Twin_Dark/Meshes/F_MED_Twin_Dark_Jacket_Physics
bodies {
 dyn_parent_shirt_bck [{shapeType=Capsule, length=0.505, radius=0.505, rotation=<rot 0 -90 0>}]
 dyn_parent_shirt_bck_l [{shapeType=Capsule, length=0.505, radius=0.505, rotation=<rot 0 -90 0>}]
 dyn_parent_shirt_bck_r [{shapeType=Capsule, length=0.505, radius=0.505, rotation=<rot 0 -90 0>}]
 dyn_shirt_bck_01 [{shapeType=Capsule, length=4.25, radius=4.505, center=<5 0 0>, rotation=<rot 90 0 0>}]
 dyn_shirt_bck_01_l [{shapeType=Capsule, length=5.25, radius=3.505, center=<-5 0 0>, rotation=<rot 90 0 0>}]
 dyn_shirt_bck_01_r [{shapeType=Capsule, length=4.25, radius=4.505, center=<5 0 0>, rotation=<rot 90 0 0>}]
}
constraints [
 {bone1Name=dyn_shirt_bck_01_r, bone2Name=dyn_parent_shirt_bck_r, swing1LimitDegrees=35, swing2LimitDegrees=30, twistLimitDegrees=5}
 {bone1Name=dyn_shirt_bck_01_l, bone2Name=dyn_parent_shirt_bck_l, swing1LimitDegrees=35, swing2LimitDegrees=30, twistLimitDegrees=5}
 {bone1Name=dyn_shirt_bck_01, bone2Name=dyn_parent_shirt_bck, swing1LimitDegrees=35, swing2LimitDegrees=30, twistLimitDegrees=5}
 {
  bone1Name=dyn_shirt_bck_01_l
  bone2Name=dyn_shirt_bck_01
  swing1LimitDegrees=45
  swing1Motion=Free
  swing2LimitDegrees=45
  swing2Motion=Free
  twistLimitDegrees=45
  twistMotion=Free
 }
 {
  bone1Name=dyn_shirt_bck_01_r
  bone2Name=dyn_shirt_bck_01
  swing1LimitDegrees=45
  swing1Motion=Free
  swing2LimitDegrees=45
  swing2Motion=Free
  twistLimitDegrees=45
  twistMotion=Free
 }
]
