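# Animation Blueprint for the F_MED_Twin_Dark female body cosmetic, adding corrective bone motion and cloth-bone physics to the pose copied from its attached parent mesh.
# Thigh and head twist, measured against the reference pose, drive additive glute, hip and collar offsets, with hip bones also constrained between pelvis and thigh twist bones.
# Two RigidBody passes then simulate jacket and collar.
# Above a 2D speed threshold, thigh correctives and collar physics are dampened.
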AnimBlueprint /BRCosmetics/Characters/Player/Female/Medium/Bodies/F_MED_Twin_Dark/Meshes/F_MED_Twin_Dark_AnimBP
target_skeleton=&/F_MED_Twin_Dark_Skeleton
graphs [
 AnimGraph {nodes [
   n1 = "Transform (Modify) Bone - Bone: dyn_glute_l"(
    ComponentPose=Local To Component(LocalPose=Copy Pose From Mesh() {bUseAttachedParent=true})
    Alpha=Get thigh_rot_l()
    Translation="-3.750,0.875,-0.9"
    Rotation="0.000,0.000,-55"
    Scale="1.25,1.25,1.25"
   ) {
    rotationMode=BMM_Additive
    rotationSpace=BCS_BoneSpace
    scaleMode=BMM_Additive
    scaleSpace=BCS_BoneSpace
    translationMode=BMM_Additive
    translationSpace=BCS_BoneSpace
    alphaBoolBlend {blendOption=Linear}
    boneToModify {boneName=dyn_glute_l}
    rotation {roll=-55}
    scale=<1.25 1.25 1.25>
    translation=<-3.75 0.875 -0.9>
   }
   n2 = "Transform (Modify) Bone - Bone: dyn_glute_r"(ComponentPose=@n1, Alpha=Get thigh_rot_r(), Translation="-3.75,0,0", Rotation="0.000,0.000,-55", Scale="1.3,1.3,1.3") {
    rotationMode=BMM_Additive
    rotationSpace=BCS_BoneSpace
    scaleMode=BMM_Additive
    scaleSpace=BCS_BoneSpace
    translationMode=BMM_Additive
    translationSpace=BCS_BoneSpace
    alphaBoolBlend {blendOption=Linear}
    boneToModify {boneName=dyn_glute_r}
    rotation {roll=-55}
    scale=<1.3 1.3 1.3>
    translation {x=-3.75}
   }
   n3 = "Constraint - dyn_hip_offset_l (PP)"!(ComponentPose=@n2, Alpha=1, ConstraintWeights_0=0.6, ConstraintWeights_1:float=0.4) -> (Pose) {
    alphaBoolBlend {blendOption=Linear}
    boneToModify {boneName=dyn_hip_offset_l}
    constraintSetup [{targetBone {boneName=pelvis}}, {targetBone {boneName=thigh_twist_01_l}}]
    constraintWeights [1,1]
   }
   n4 = "Constraint - dyn_hip_offset_r (PP)"!(ComponentPose=@n3, Alpha=1, ConstraintWeights_0=0.6, ConstraintWeights_1:float=0.4) -> (Pose) {
    alphaBoolBlend {blendOption=Linear}
    boneToModify {boneName=dyn_hip_offset_r}
    constraintSetup [{targetBone {boneName=pelvis}}, {targetBone {boneName=thigh_twist_01_r}}]
    constraintWeights [1,1]
   }
   n5 = "Transform (Modify) Bone - Bone: dyn_hip_offset_l"(ComponentPose=@n4, Alpha=Get thigh_rot_l(), Translation="3,18,8", Rotation="0,0,0", Scale="1,1,1") {
    rotationMode=BMM_Additive
    rotationSpace=BCS_BoneSpace
    scaleMode=BMM_Additive
    scaleSpace=BCS_BoneSpace
    translationMode=BMM_Additive
    translationSpace=BCS_BoneSpace
    alphaBoolBlend {blendOption=Linear}
    boneToModify {boneName=dyn_hip_offset_l}
    translation=<3 18 8>
   }
   n6 = "Transform (Modify) Bone - Bone: dyn_hip_offset_r"(ComponentPose=@n5, Alpha=Get thigh_rot_r(), Translation="-0.000,-3.336,-14.885", Rotation="-10,0.000,0", Scale="1.5,1.5,1.5") {
    rotationMode=BMM_Additive
    rotationSpace=BCS_BoneSpace
    scaleMode=BMM_Additive
    scaleSpace=BCS_BoneSpace
    translationMode=BMM_Additive
    translationSpace=BCS_BoneSpace
    alphaBoolBlend {blendOption=Linear}
    boneToModify {boneName=dyn_hip_offset_r}
    rotation {pitch=-10}
    scale=<1.5 1.5 1.5>
    translation {y=-3.336, z=-14.885}
   }
   n7 = "Transform (Modify) Bone - Bone: dyn_collar_bck_01"(
    ComponentPose=@n6
    Alpha=Get collar_rot_bck() -> (collar_rot_bck:double)
    Translation="-5,1,0.000"
    Rotation="0.000,-80,0.000"
    Scale="1,1,1"
   ) {
    rotationMode=BMM_Additive
    rotationSpace=BCS_BoneSpace
    translationMode=BMM_Additive
    translationSpace=BCS_BoneSpace
    alphaBoolBlend {blendOption=Linear}
    boneToModify {boneName=dyn_collar_bck_01}
    rotation {yaw=-80}
    translation=<-5 1>
   }
   n8 = Get collar_rot_l() -> (collar_rot_l:double)
   n9 = "Transform (Modify) Bone - Bone: dyn_driver_collar_back_l"(ComponentPose=@n7, Alpha=@n8, Translation="-0.946,2.030,-0.000", Rotation="-4.044,-55.000,0.000", Scale="1,1,1") {
    rotationMode=BMM_Additive
    rotationSpace=BCS_BoneSpace
    translationMode=BMM_Additive
    translationSpace=BCS_BoneSpace
    alphaBoolBlend {blendOption=Linear}
    boneToModify {boneName=dyn_driver_collar_back_l}
    rotation {pitch=-4.044, yaw=-55}
    translation=<-0.946 2.03>
   }
   n10 = "Transform (Modify) Bone - Bone: dyn_driver_collarA_l"(ComponentPose=@n9, Alpha=@n8, Translation="-1.445,1.309,-0.000", Rotation="0.000,-33.970,-17.013", Scale="1,1,1") {
    rotationMode=BMM_Additive
    rotationSpace=BCS_BoneSpace
    translationMode=BMM_Additive
    translationSpace=BCS_BoneSpace
    alphaBoolBlend {blendOption=Linear}
    boneToModify {boneName=dyn_driver_collarA_l}
    rotation {roll=-17.0131, yaw=-33.9701}
    translation=<-1.445 1.309>
   }
   n11 = RigidBody(ComponentPose=@n10) {
    bClampLinearTranslationLimitToRefPose=true
    bOverrideWorldGravity=true
    overridePhysicsAsset=&/F_MED_Twin_Dark_Jacket_Physics
    simulationSpace=BaseBoneSpace
    alphaBoolBlend {blendOption=Linear}
    baseBoneRef {boneName=root}
    componentAppliedLinearAccClamp=<1100 1100 1100>
    componentLinearAccScale=<0.5 0.5 0.5>
    componentLinearVelScale=<0.5 0.5 0.5>
    overrideWorldGravity {z=-980}
   }
   n12 = RigidBody(ComponentPose=@n11, Alpha=Get collar_alpha() -> (collar_alpha:double)) {
    bClampLinearTranslationLimitToRefPose=true
    bOverrideWorldGravity=true
    overridePhysicsAsset=&/F_MED_Twin_Dark_Collar_Physics
    simulationSpace=BaseBoneSpace
    alphaBoolBlend {blendOption=Linear}
    baseBoneRef {boneName=root}
    componentAppliedLinearAccClamp=<1100 1100 1100>
    componentLinearAccScale=<0.5 0.5 0.5>
    componentLinearVelScale=<0.5 0.5 0.5>
    overrideWorldGravity {z=-980}
   }
   Output Pose(Result=Component To Local(ComponentPose=@n12)) {layerGroup=None}
  ]}
 EventGraph {nodes [
   Event BlueprintUpdateAnimation()
   > Get!(execute:Exec) -> (then:Exec, else:Exec, Skeletal_Mesh:Skeletal Mesh Component Object Reference)
   > n1 = Sequence!(execute) -> (then_0, then_1, then_2:Exec, then_3:Exec)
   n2 = Branch(execute=@n1.then_3, Condition="float > float"(A=Get PawnSpeed2D(), B=100) {is=title K2Node_CallFunction})
   > Set collar_alpha(collar_alpha:double=0.1) -> (Output_Get:double)
   Set collar_alpha(execute=@n2.else, collar_alpha=1)
   Set collar_rot_bck(
    execute=@n1.then_1
    collar_rot_bck:double=GetAlphaFromReferenceTransformDelta(
     SkeletalMeshComponent=Get Skeletal_Mesh() -> (Skeletal_Mesh:Skeletal Mesh Component Object Reference)
     BoneName=head
     BaseName=spine_05
     DeltaComponentType=QuaternionTwist
     InputRange=(X=7,Y=75)
     TwistAxis="0,0,1"
    )
   ) -> (Output_Get:double)
   Set collar_rot_l(
    execute=@n1.then_2
    collar_rot_l:double=GetAlphaFromReferenceTransformDelta(
     SkeletalMeshComponent=Get Skeletal_Mesh()
     BoneName=head
     BaseName=spine_05
     DeltaComponentType=QuaternionTwist
     InputRange=(X=7,Y=75)
     TwistAxis="1,0,0"
    )
   ) -> (Output_Get:double)
   n3 = Get Skeletal_Mesh()
   n4 = GetAlphaFromReferenceTransformDelta(SkeletalMeshComponent=@n3, BoneName=thigh_l, BaseName=pelvis, DeltaComponentType=QuaternionTwist, InputRange=(X=20,Y=100), TwistAxis="0,0,1")
   Set thigh_rot_l(execute=@n1.then_0, thigh_rot_l=@n4)
   n5 = GetAlphaFromReferenceTransformDelta(SkeletalMeshComponent=@n3, BoneName=thigh_r, BaseName=pelvis, DeltaComponentType=QuaternionTwist, InputRange=(X=20,Y=100), TwistAxis="0,0,1")
   > Set thigh_rot_r(thigh_rot_r=@n5)
   > Branch(Condition="float > float"(A=Get PawnSpeed2D(), B=100))
   > Set thigh_rot_l(thigh_rot_l=float * float(A=@n4, B=0.35))
   > Set thigh_rot_r(thigh_rot_r=float * float(A=@n5, B=0.35))
   Event BlueprintInitializeAnimation()
   > Set Skeletal_Mesh(Skeletal_Mesh:Skeletal Mesh Component Object Reference=GetOwnerSkeletalMeshComponent()) -> (Output_Get:Skeletal Mesh Component Object Reference)
  ]}
]
variables [
 collar_alpha {type=number}
 collar_rot_bck {type=number}
 collar_rot_l {type=number}
 Skeletal_Mesh {type=SkeletalMeshComponent}
 thigh_rot_l {type=number}
 thigh_rot_r {type=number}
]
